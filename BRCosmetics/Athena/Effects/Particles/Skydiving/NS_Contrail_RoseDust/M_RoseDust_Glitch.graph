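# Translucent unlit Niagara mesh-particle material for glitchy shards in the Rose Dust skydiving contrail.
# Emissive hue-shifts particle color per piece, adds a mask texture and brightens with a scene-normal Fresnel near intersections.
# Opacity cycles over time between rotated screen-space scanlines, dots, circle-fade and flat fill, using vertex color and dynamic randoms as offsets, with depth fading.
# Offset sine waves build a world-position jiggle; quality and shading-path switches simplify cheaper platforms.
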Material /BRCosmetics/Athena/Effects/Particles/Skydiving/NS_Contrail_RoseDust/M_RoseDust_Glitch
graph {
 properties {
  bUsedWithNiagaraMeshParticles=true
  blendMode=BLEND_Translucent
  pixelDepthOffsetMode=PDOM_Legacy
  refractionMethod=RM_IndexOfRefraction
  shadingModel=MSM_Unlit
  usedShadingModels=Unlit
 }
 nodes [
  Constant3Vector() {constant=<rgba 0 0 1 0>}
  n1 = SceneTexture() {sceneTextureId=PPI_WorldNormal}
  n2 = Multiply(A=Fresnel(
    ExponentIn=ScalarParameter() {defaultValue=6, parameterName=WorldNormalFresnelXP}
    Normal=ComponentMask(None=@n1.Color) {b=true, g=true, r=true}
   ) {baseReflectFraction=0}) {constB=5}
  n3 = DepthFade(FadeDistance=ScalarParameter() {defaultValue=250, parameterName=VisibleDepthFade})
  n4 = LinearInterpolate(Alpha=Multiply(A=Saturate(None=@n2), B=OneMinus(None=@n3))) {constA=0.15, constB=1.25}
  n5 = Constant() {r=0.5}
  n6 = VertexColor()
  n7 = DynamicParameter() {defaultValue=<rgba 0.3 0.5 1 1>, paramNames [Random,Random2,Jitter,Param4]}
  n8 = Add(A=Subtract(A=@n6.R) {constB=0.5}, B=@n7.Random)
  n9 = ScalarParameter() {defaultValue=0.2, parameterName=MaxHueShift}
  n10 = ParticleColor()
  n11 = TextureSample() {samplerType=SAMPLERTYPE_LinearColor, texture=&/T_RoseDust_Glitch_M}
  n12 = Add(
   A=MaterialFunctionCall(Hue Shift Percentage=Multiply(A=@n8, B=@n9) {constB=0.15}, Texture=@n10.RGB) {materialFunction=&HueShift}.Result
   B=Multiply(A=MaterialFunctionCall(R=@n11.R, G=@n11.R) {materialFunction=&AppendMany}.RGB) {constB=3}
  )
  n13 = Constant() {r=-1}
  MaterialFunctionCall(Emissive=Multiply(A=ShadingPathSwitch(Default=@n4, Forward=@n5, Mobile=@n5), B=Desaturation(None=@n12, Fraction=@n13))) {materialFunction=&MF_FX_EmissiveAdjustments}
  n14 = Fmod(A=Add(A=Add(A=@n6.R, B=@n6.R), B=@n7.Random2), B=Constant() {r=1})
  n15 = DepthFade(
   Opacity=LinearInterpolate(B=Power(Base=@n14) {constExponent=5}, Alpha=@n3) {constA=1}
   FadeDistance=ScalarParameter() {defaultValue=3, parameterName=DepthFade}
  )
  n16 = Constant() {r=1}
  n17 = DynamicParameter() {defaultValue=<rgba 0.3 0.5 1 1>, paramNames [Random,Random2,Jitter,Param4]}
  n18 = Add(A=@n17.Random, B=Time())
  n19 = VertexColor()
  n20 = Add(A=@n17.Random, B=@n19.G)
  n21 = LinearInterpolate(Alpha=Fmod(A=@n20, B=Constant() {r=1})) {constA=0.2}
  n22 = Fmod(A=Add(A=VertexColor().R, B=Multiply(A=@n18, B=@n21) {constB=0.4}), B=Constant() {r=2})
  n23 = MaterialFunctionCall(Float2=ViewSize()) {materialFunction=&BreakOutFloat2Components}
  n24 = Divide(A=@n23.R, B=@n23.R)
  n25 = Constant() {r=1}
  n26 = Rotator(Coordinate=Multiply(A=AppendVector(A=@n24, B=@n25), B=ScreenPosition().ViewportUV), Time=@n20) {speed=8}
  n27 = ComponentMask(None=@n26) {g=true}
  n28 = ScalarParameter() {defaultValue=150, parameterName=ScanlineTile}
  n29 = ConstantBiasScale(None=Sine(None=Multiply(A=@n27, B=@n28) {constB=50}))
  n30 = Fmod(A=Multiply(A=@n26, B=ScalarParameter() {defaultValue=100, parameterName=DotsTile}), B=Constant() {r=1})
  n31 = Constant() {r=0.5}
  n32 = ScalarParameter() {defaultValue=0.4, parameterName=DotRadius1}
  n33 = TextureSample(UVs=Multiply(A=@n26, B=ScalarParameter() {defaultValue=25, parameterName=DotsTile2})) {samplerType=SAMPLERTYPE_LinearColor, texture=/Game/Weapons/WeaponSkins/Wraps/Textures/CircleFade/T_Wrap_CircleFade_M}
  n34 = If(
   A=Constant() {r=0.5}
   B=@n22
   "A > B"=Saturate(None=Multiply(A=@n29, B=@n29) {constB=2})
   "A == B"=LinearInterpolate(Alpha=SphereMask(A=@n30, B=@n31, Radius=@n32) {hardnessPercent=75}) {constA=0.5}
   "A < B"=MaterialFunctionCall(In=@n33.G, Contrast=Constant() {r=1}) {materialFunction=&CheapContrast}.Result
  ) {equalsThreshold=0.25}
  n35 = Constant() {r=1}
  n36 = Constant()
  n37 = If(A=@n22, B=@n35, "A > B"=@n35, "A == B"=@n36, "A < B"=@n36) {equalsThreshold=0}
  Multiply(
   A=Multiply(A=@n10.A, B=@n11.G)
   B=Multiply(A=QualitySwitch(Default=@n15, Low=@n16), B=LinearInterpolate(A=@n34, Alpha=@n37) {constA=0.6, constB=0.6})
  )
  Multiply(A=@n33.G, B=@n33.G)
  n38 = Multiply(A=Add(A=Add(A=@n17.Random, B=@n17.Random), B=@n19.R)) {constB=2}
  n39 = Add(A=Multiply(A=ScalarParameter() {defaultValue=2, parameterName=WPOJiggleSpeed}, B=Time()), B=@n38)
  n40 = Sine(None=@n39) {period=0.3}
  n41 = Sine(None=@n39) {period=0.375}
  n42 = Sine(None=@n39) {period=0.291}
  n43 = Fmod(A=@n38, B=Constant() {r=1})
  n44 = ScalarParameter() {defaultValue=2, parameterName=WPOScale}
  NamedRerouteDeclaration(None=Multiply(
    A=MaterialFunctionCall(R=@n40, G=@n41, B=@n42) {materialFunction=&AppendMany}.RGB
    B=Multiply(A=@n43, B=@n44)
   )) {name=WPO, nodeColor=<rgba 1 0.6094 0 1>}
  NamedRerouteUsage()
  OneMinus(None=@n3)
  n45 = Fmod(A=Add(A=@n6.G, B=@n7.Random2), B=Constant() {r=1})
  n46 = Multiply(B=@n37) {constA=1}
  QualitySwitch(
   Default=LinearInterpolate(B=LinearInterpolate(Alpha=@n45) {constA=0.75, constB=1.35}, Alpha=Round(None=@n46)) {constA=1}
   Low=Constant() {r=1}
  )
 ]
}
