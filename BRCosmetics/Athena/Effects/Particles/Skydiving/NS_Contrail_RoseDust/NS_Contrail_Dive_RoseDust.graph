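# Niagara skydiving contrail for the RoseDust cosmetic's dive, trailing cyan, pink and purple energy from the character's skeletal mesh.
# Paired waving ribbons seed crackle sprites that sample their color alpha for intensity.
# Bone-attached emitters add noisy hand and body sparks, flickering glow sprites with lights, spinning spline meshes and jittering glitch shapes.
# A system Dive value scales spline and glitch spawn rates, while an integrator value drives ribbon velocity.
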NiagaraSystem /BRCosmetics/Athena/Effects/Particles/Skydiving/NS_Contrail_RoseDust/NS_Contrail_Dive_RoseDust
dependencies {
 usedDataInterfaces [
  &NiagaraDataInterfaceCurve
  &NiagaraDataInterfaceParticleRead
  &NiagaraDataInterfaceArrayColor
  &NiagaraDataInterfaceArrayInt32
  &NiagaraDataInterfaceArrayFloat3
  &NiagaraDataInterfaceVectorCurve
  &NiagaraDataInterfaceArrayFloat
 ]
 usedDynamicInputs [&FN_FrontEndVector3Modifier]
 usedModules [
  &SystemState
  /Game/Effects/Niagara/Modules/Skydiving/FN_DivingAttributes/v5/FN_DivingAttributes
  &FN_Integrator
  &self:SystemSpawnScript.SystemScriptSource.SystemScriptGraph.NiagaraNodeAssignment_0.SetVariables_#1
  &self:SystemSpawnScript.SystemScriptSource.SystemScriptGraph.NiagaraNodeAssignment_1.SetVariables_#2
 ]
}
emitters {
 BodyCrackle {
  emitterUpdateScript [
   &EmitterState()
   &SpawnRate(SpawnRate=70, Use Spawn Probability=false)
  ]
  particleSpawnScript [
   &FN_SetSpawnGroup(Number of Spawn Groups=2)
   &InitializeParticle(
    Color=&RandomRangeLinearColor(Minimum=<rgba 0.0499 0.007 0.25 1>, Maximum=<rgba 0.7231 0.0513 0.2346 1>)
    Lifetime Max=0.45
    Lifetime Min=0.3
    Lifetime Mode=::NewEnumerator1
    Sprite Size Mode=::NewEnumerator3
    Uniform Sprite Size=&Lerp_Float(
     A=2
     B=4
     Alpha=&FloatFromCurve(
      FloatCurve=DataInterface() {propertyValues {Curve {keys [
          {interpMode=RCIM_Cubic, tangentMode=RCTM_User, time=0, value=0}
          {arriveTangent=2.3987, interpMode=RCIM_Cubic, leaveTangent=2.3987, tangentMode=RCTM_User, time=0.6, value=0.55}
          {interpMode=RCIM_Cubic, tangentMode=RCTM_User, time=1, value=1}
         ]}}}
      CurveIndex=&RandomRangeFloat(Minimum=0, Maximum=1)
     )
    )
    UsePositionOffset=false
    Write Material Random=true
   )
   &SkeletalMeshLocation(
    Bone / Socket Index=&MaskIntBySpawnGroup(Masked Int=5, Passthrough Int=6, Spawn Group Mask=0)
    Bone Sampling Mode=::NewEnumerator4
    Interpolate Mesh Sampling=false
    MaskSpawnGroup=false
    Orientation Sampling=::NewEnumerator1
    Position Sampling=::NewEnumerator1
    Sampled Position Offset=&MaskVectorBySpawnGroup(
     Masked Vector=&RandomRangeVector(Minimum=<-3 0 0>, Maximum=<25 0 0>)
     Passthrough Vector=&RandomRangeVector(Minimum=<-25 0 0>, Maximum=<3 0 0>)
    )
    Skeletal Mesh=$User.SkeletalMesh
   )
   &SphereLocation(
    MaskSpawnGroup=false
    Override Seed=false
    Sphere Radius=2
    Surface Only=false
    UseNonUniformScale=false
    UseOffset=false
    UseSphereOrientation=false
   )
   &CameraOffset(Camera Offset Amount=3)
  ]
  particleUpdateScript [
   &ParticleState()
   &VectorNoiseForce(
    Force Amount=&FloatFromCurve(
     FloatCurve=DataInterface() {propertyValues {Curve {keys [{interpMode=RCIM_Cubic, tangentMode=RCTM_User, time=0, value=1}, {interpMode=RCIM_Cubic, tangentMode=RCTM_User, time=1, value=0}]}}}
     Scale Curve=700
    )
    Module.Use Random Noise Time Interval=false
    Random Noise Time Interval=0.15
   )
   &WindForce(
    Wind Speed=&TransformVector(Vector=<0 0 1>, Source Space=::World, Destination Space=::Local)
    Wind Speed Scale=800
   )
   &Drag(Drag=7, Use Linear Drag=true, Use Rotational Drag=false)
   &ScaleSpriteSize(
    Uniform Curve Scale=&Waveform(
     Global Phase=$Particles.Age
     "Waveform [1]"=::NewEnumerator12
     "[1] Frequency"=&RandomRangeFloat(Minimum=15, Maximum=40)
     "[1] Amplitude Min/Max"=<0.5 1.5>
    )
    Uniform Curve Sprite Scale=DataInterface() {propertyValues {Curve {keys [
        {arriveTangent=10.8984, interpMode=RCIM_Cubic, leaveTangent=10.8984, tangentMode=RCTM_User, time=0, value=0}
        <at 0.2 1>
        {interpMode=RCIM_Cubic, tangentMode=RCTM_User, time=0.6, value=1}
        {arriveTangent=-5.0292, interpMode=RCIM_Cubic, leaveTangent=-5.0292, tangentMode=RCTM_User, time=1, value=0}
       ]}}}
   )
   /SetVariables(
    Particles.SpriteParams0:Vector4f=&MakeFloat4FromVectorAndFloat("Vector (RGB)"=&MakeVectorFromLinearColorRGB(Color=<rgba 0 0.8192 1 1>), "Float (Alpha)"=1)
    Particles.SpriteParams1:Vector4f=&MakeFloat4FromVectorAndFloat(
     "Vector (RGB)"=&Subtract_Vector(A=$Particles.SkeletalMeshLocation.SampledPosition, B=Linked() {linkedVariable {name=Particles.Position, type=NiagaraPosition}})
     "Float (Alpha)"=2
    )
   ) {isSetParameters=true, module=SetVariables_#3}
   &SolveForcesAndVelocity(Clamp Velocity=false, Limit Acceleration=false)
  ]
  rendererClasses [&NiagaraSpriteRendererProperties]
  renderers [{rendererClass=&NiagaraSpriteRendererProperties}]
 }
 GlitchShapes {
  emitterUpdateScript [
   &EmitterState()
   &SpawnRate(SpawnRate=12, Use Spawn Probability=false)
   &SpawnRate(SpawnGroup=1, SpawnRate=&Lerp_Float(A=8, B=17, Alpha=$System.Dive), Use Spawn Probability=false) {module=SpawnRate001}
  ]
  particleSpawnScript [
   /SetVariables(Particles.RandomBone:NiagaraInt32=&RandomRangeInt(Minimum=0, Maximum=5)) {isSetParameters=true}
   /SetVariables(Particles.BoneIndex:NiagaraInt32=&MaskIntBySpawnGroup(
     Masked Int=&RandomRangeInt(Minimum=0, Maximum=1)
     Passthrough Int=&SelectIntFromArray(Direct Array Index=$Particles.RandomBone, Int Selection Array=DataInterface() {propertyValues {IntData [0,1,5,6,9,10]}})
     Spawn Group Mask=1
    )) {isSetParameters=true}
   &InitializeParticle(
    Color=&RandomRangeLinearColor(Minimum=<rgba 0.9062 0.332 2 1>, Maximum=<rgba 0.0833 2.5136 4 1>)
    Lifetime Max=0.35
    Lifetime Min=0.1
    Lifetime Mode=::NewEnumerator1
    Mesh Scale=&Multiply_VectorByFloat(
     Vector=&RandomRangeVector(Minimum=<0.1 0.1 0.1>, Maximum=<0.4 0.4 0.4>)
     Float=&SelectFloatFromArray(Direct Array Index=$Particles.RandomBone, Float Selection Array=DataInterface() {propertyValues {FloatData [1.2,1.2,0.8,0.8,0.5,0.5]}})
    )
    Mesh Scale Mode=::NewEnumerator4
    Sprite Size Mode=::NewEnumerator3
    UsePositionOffset=false
    Write Material Random=true
   )
   &InitialMeshOrientation(Use Rotation Vector=false)
   &SkeletalMeshLocation(
    Bone / Socket Index=$Particles.BoneIndex
    Bone Sampling Mode=::NewEnumerator4
    MaskSpawnGroup=false
    Orientation Sampling=::NewEnumerator1
    Skeletal Mesh=$User.SkeletalMesh
   )
   &CameraOffset(Camera Offset Amount=8)
   &AddVelocity(
    Velocity=<0 0 1>
    Velocity Speed Scale=&MaskFloatBySpawnGroup(
     Masked Float=&RandomRangeFloat(Minimum=200, Maximum=300)
     Passthrough Float=&RandomRangeFloat(Minimum=0, Maximum=40)
     Spawn Group Mask=1
    )
   )
   /SetVariables(Particles.Position=&Add_Vector(
     A=$Particles.SkeletalMeshLocation.SampledPosition
     B=&RandomRangeVector(Minimum=<-5 -5 -17>, Maximum=<5 5 -5>)
    )) {isSetParameters=true}
   /SetVariables(Particles.MeshIndex=&RandomRangeInt(Minimum=0, Maximum=1)) {enabled=false, isSetParameters=true}
  ]
  particleUpdateScript [
   &ParticleState()
   /SetVariables(Particles.Jitter:NiagaraFloat=&ClampFloat(
     Float=&Waveform(
      Global Clamp Min=-1
      "Waveform [1]"=::NewEnumerator12
      "[1] Frequency"=&RandomRangeFloat(Minimum=15, Maximum=25)
      "[1] Amplitude Min/Max"=<-150 150>
     )
     Min=-0.25
     Max=0.25
    )) {isSetParameters=true}
   &DynamicMaterialParameters(
    Index 0 Param 1=&Add_Float(A=$Particles.Jitter, B=&RandomRangeFloat(Minimum=0, Maximum=1))
    Index 0 Param 2=&Subtract_Float(A=&RandomRangeFloat(Minimum=0, Maximum=1), B=$Particles.Jitter)
    Index 0 Param 3=$Particles.Jitter
    Module.Param0WriteEnabled=true
    Module.Param1WriteEnabled=true
    Module.Param2WriteEnabled=true
    Module.Param3WriteEnabled=false
   )
   &ScaleColor(
    Scale Alpha=&Add_Float(A=1, B=&ClampFloat(Float=$Particles.Jitter))
    ScaleA=true
    ScaleRGB=true
   )
   &UpdateMeshOrientation(Rotation Rate=0.2, Rotation Vector=&RandomRangeVector(Minimum=<0 0 0>, Maximum=<1 1 1>))
   &SolveForcesAndVelocity(Clamp Velocity=false, Limit Acceleration=false)
  ]
  rendererClasses [&NiagaraMeshRendererProperties]
  renderers [{rendererClass=&NiagaraMeshRendererProperties, rendererIndex=0}]
 }
 HandCrackle {
  emitterUpdateScript [
   &EmitterState()
   &SpawnRate(SpawnRate=200, Use Spawn Probability=false)
  ]
  particleSpawnScript [
   &FN_SetSpawnGroup(Number of Spawn Groups=2)
   &InitializeParticle(
    Lifetime Max=0.5
    Lifetime Min=0.3
    Lifetime Mode=::NewEnumerator1
    Sprite Size Mode=::NewEnumerator3
    Uniform Sprite Size=&Lerp_Float(
     A=3
     B=6
     Alpha=&FloatFromCurve(
      FloatCurve=DataInterface() {propertyValues {Curve {keys [
          {interpMode=RCIM_Cubic, tangentMode=RCTM_User, time=0, value=0}
          {arriveTangent=2.3987, interpMode=RCIM_Cubic, leaveTangent=2.3987, tangentMode=RCTM_User, time=0.6, value=0.55}
          {interpMode=RCIM_Cubic, tangentMode=RCTM_User, time=1, value=1}
         ]}}}
      CurveIndex=&RandomRangeFloat(Minimum=0, Maximum=1)
     )
    )
    UsePositionOffset=false
    Write Material Random=true
   )
   &SkeletalMeshLocation(
    Bone / Socket Index=&SpawnGroup()
    Bone Sampling Mode=::NewEnumerator4
    Interpolate Mesh Sampling=false
    MaskSpawnGroup=false
    Orientation Sampling=::NewEnumerator1
    Position Sampling=::NewEnumerator1
    Sampled Position Offset=&MaskVectorBySpawnGroup(Masked Vector=<8 0 0>, Passthrough Vector=<-8 0 0>)
    Skeletal Mesh=$User.SkeletalMesh
   )
   &SphereLocation(
    MaskSpawnGroup=false
    Override Seed=false
    Sphere Radius=3
    Surface Only=true
    UseNonUniformScale=false
    UseOffset=false
    UseSphereOrientation=false
   )
   &AddVelocity(
    Module.Use Falloff Distance=false
    Override Seed=false
    Velocity Mode=::NewEnumerator1
    Velocity Origin:Vector3f=$Particles.SkeletalMeshLocation.SampledPosition
    Velocity Speed=60
   )
   &CameraOffset(Camera Offset Amount=3)
  ]
  particleUpdateScript [
   &ParticleState()
   &VectorNoiseForce(
    Force Amount=&FloatFromCurve(
     FloatCurve=DataInterface() {propertyValues {Curve {keys [{interpMode=RCIM_Cubic, tangentMode=RCTM_User, time=0, value=1}, {interpMode=RCIM_Cubic, tangentMode=RCTM_User, time=1, value=0}]}}}
     Scale Curve=1500
    )
    Module.Use Random Noise Time Interval=false
    Random Noise Time Interval=0.15
   )
   &WindForce(
    Wind Speed=&TransformVector(Vector=<0 0 1>, Source Space=::World, Destination Space=::Local)
    Wind Speed Scale=800
   )
   /Niagara/Modules/Update/Forces/LineAttractionForce(
    Attraction Strength=500
    Line End:Vector3f=&Add_Vector(A=$Particles.Initial.SkeletalMeshLocation.SampledPosition, B=<0 0 500>)
    Line Start:Vector3f=$Particles.Initial.SkeletalMeshLocation.SampledPosition
    Use Attraction Falloff=false
   )
   &Drag(Drag=7, Use Linear Drag=true, Use Rotational Drag=false)
   &ScaleSpriteSize(
    Uniform Curve Scale=&Waveform(
     Global Phase=$Particles.Age
     "Waveform [1]"=::NewEnumerator12
     "[1] Frequency"=&RandomRangeFloat(Minimum=15, Maximum=40)
     "[1] Amplitude Min/Max"=<0.5 1.5>
    )
    Uniform Curve Sprite Scale=DataInterface() {propertyValues {Curve {keys [
        {arriveTangent=10.8984, interpMode=RCIM_Cubic, leaveTangent=10.8984, tangentMode=RCTM_User, time=0, value=0}
        <at 0.2 1>
        {interpMode=RCIM_Cubic, tangentMode=RCTM_User, time=0.6, value=1}
        {arriveTangent=-5.0292, interpMode=RCIM_Cubic, leaveTangent=-5.0292, tangentMode=RCTM_User, time=1, value=0}
       ]}}}
   )
   /SetVariables(
    Particles.SpriteParams0:Vector4f=&MakeFloat4FromVectorAndFloat("Vector (RGB)"=&MakeVectorFromLinearColorRGB(Color=<rgba 0 2.4575 3 1>), "Float (Alpha)"=1)
    Particles.SpriteParams1:Vector4f=&MakeFloat4FromVectorAndFloat(
     "Vector (RGB)"=&Subtract_Vector(A=$Particles.SkeletalMeshLocation.SampledPosition, B=Linked() {linkedVariable {name=Particles.Position, type=NiagaraPosition}})
     "Float (Alpha)"=2
    )
   ) {isSetParameters=true, module=SetVariables_#3}
   &SolveForcesAndVelocity(Clamp Velocity=false, Limit Acceleration=false)
   &ScaleColor(
    Scale RGB=&MakeVectorFromLinearColorRGB(Color=&LerpLinearColors(
      StartColor=<rgba 0.0999 0.0141 0.5 1>
      EndColor=<rgba 0.7231 0.0513 0.2346 1>
      LerpFactor=&FloatFromCurve(
       FloatCurve=DataInterface() {propertyValues {Curve {keys [{interpMode=RCIM_Cubic, tangentMode=RCTM_User, time=0, value=0}, {interpMode=RCIM_Cubic, tangentMode=RCTM_User, time=1, value=1}]}}}
       CurveIndex=&NormalizeDistanceRange(
        Start Position:NiagaraPosition=$Particles.Position
        End Position:NiagaraPosition=Linked() {linkedVariable {name=Particles.SkeletalMeshLocation.SampledPosition, type=Vector3f}}
        Distance:NiagaraFloat=30
        Invert Normalized Range:NiagaraBool=false
       )
      )
     ))
    ScaleA=true
    ScaleRGB=true
   )
  ]
  rendererClasses [&NiagaraSpriteRendererProperties]
  renderers [{rendererClass=&NiagaraSpriteRendererProperties}]
 }
 HandCrackleInner {
  emitterUpdateScript [
   &EmitterState()
   &SpawnRate(SpawnRate=200, Use Spawn Probability=false)
  ]
  particleSpawnScript [
   &FN_SetSpawnGroup(Number of Spawn Groups=2)
   &InitializeParticle(
    Color Maximum=<rgba 0.0036 2.6109 3 1>
    Color Minimum=<rgba 0 1.2291 2 1>
    Color Mode=::NewEnumerator2
    Lifetime Max=0.4
    Lifetime Min=0.25
    Lifetime Mode=::NewEnumerator1
    Sprite Size Mode=::NewEnumerator1
    Uniform Sprite Size Max=5
    Uniform Sprite Size Min=3
    UsePositionOffset=false
    Write Material Random=true
   )
   &SkeletalMeshLocation(
    Bone / Socket Index=&SpawnGroup()
    Bone Sampling Mode=::NewEnumerator4
    Interpolate Mesh Sampling=false
    MaskSpawnGroup=false
    Position Sampling=::NewEnumerator1
    Sampled Position Offset=&MaskVectorBySpawnGroup(Masked Vector=<8 0 0>, Passthrough Vector=<-8 0 0>)
    Skeletal Mesh=$User.SkeletalMesh
   )
   &SphereLocation(
    MaskSpawnGroup=false
    Override Seed=false
    Sphere Radius=4
    Surface Only=true
    UseNonUniformScale=false
    UseOffset=false
    UseSphereOrientation=false
   )
   &AddVelocity(
    Module.Use Falloff Distance=false
    Override Seed=false
    Velocity Mode=::NewEnumerator1
    Velocity Origin:Vector3f=$Particles.SkeletalMeshLocation.SampledPosition
    Velocity Speed=30
   )
  ]
  particleUpdateScript [
   &ParticleState()
   &VectorNoiseForce(
    Force Amount=&FloatFromCurve(
     FloatCurve=DataInterface() {propertyValues {Curve {keys [{interpMode=RCIM_Cubic, tangentMode=RCTM_User, time=0, value=1}, {interpMode=RCIM_Cubic, tangentMode=RCTM_User, time=1, value=0}]}}}
     Scale Curve=1500
    )
    Module.Use Random Noise Time Interval=false
    Random Noise Time Interval=0.15
   )
   &WindForce(Wind Speed=<0 0 1>, Wind Speed Scale=800)
   &LineAttractionForce(
    Attraction Strength=600
    Line End:Vector3f=&Add_Vector(A=$Particles.SkeletalMeshLocation.SampledPosition, B=<0 0 500>)
    Line Start:Vector3f=$Particles.SkeletalMeshLocation.SampledPosition
    Use Attraction Falloff=false
   )
   &Drag(Drag=7, Use Linear Drag=true, Use Rotational Drag=false)
   &ScaleSpriteSize(Uniform Curve Sprite Scale=DataInterface() {propertyValues {Curve {keys [
        {arriveTangent=10.8984, interpMode=RCIM_Cubic, leaveTangent=10.8984, tangentMode=RCTM_User, time=0, value=0}
        <at 0.2 1>
        {interpMode=RCIM_Cubic, tangentMode=RCTM_User, time=0.6, value=1}
        {arriveTangent=-5.0292, interpMode=RCIM_Cubic, leaveTangent=-5.0292, tangentMode=RCTM_User, time=1, value=0}
       ]}}})
   /SetVariables(
    Particles.SpriteParams0:Vector4f=&MakeFloat4FromVectorAndFloat("Vector (RGB)"=&MakeVectorFromLinearColorRGB(Color=<rgba 5 0.4666 0 1>), "Float (Alpha)"=1)
    Particles.SpriteParams1:Vector4f=&MakeFloat4FromVectorAndFloat(
     "Vector (RGB)"=&Subtract_Vector(A=$Particles.SkeletalMeshLocation.SampledPosition, B=Linked() {linkedVariable {name=Particles.Position, type=NiagaraPosition}})
     "Float (Alpha)"=2
    )
   ) {isSetParameters=true, module=SetVariables_#3}
   &SolveForcesAndVelocity(Clamp Velocity=false, Limit Acceleration=false)
  ]
  rendererClasses [&NiagaraSpriteRendererProperties]
  renderers [{rendererClass=&NiagaraSpriteRendererProperties}]
 }
 HandGlow {
  emitterUpdateScript [
   &EmitterState(
    Enable Distance Culling=true
    Inactive Response=::NewEnumerator1
    Life Cycle Mode=::NewEnumerator1
    Loop Behavior=::NewEnumerator1
    Loop Duration Mode=::NewEnumerator1
    MaxDistance=2000
    Module.bUseMaxDistance=true
    Module.bUseMinDistance=false
    Scalability Mode=::NewEnumerator1
    UseLoopDelay=false
   )
   &SpawnBurst_Instantaneous(Spawn Count=4, Use Spawn Probability=false, UseLoopCountLimit=false)
  ]
  particleSpawnScript [
   &FN_SetSpawnGroup()
   &InitializeParticle(
    Color=<rgba 0 2.9851 5 0.15>
    Lifetime=5
    Sprite Size Mode=::NewEnumerator3
    Uniform Sprite Size=50
    UsePositionOffset=false
    Write Material Random=true
   )
   &CameraOffset(Camera Offset Amount=6)
   /SetVariables(
    Particles.DecalCamOffset=50
    Particles.DecalColor:LinearColor=<rgba 1 1 1 10>
    Particles.DecalSize:Vector2f=&Vector2DFromFloat(Value=50)
   ) {isSetParameters=true}
   /SetVariables(Particles.LightEnabled:NiagaraBool=&SetBoolByIntComparison(A=&SpawnGroup(), B=1, Comparison Type=::NewEnumerator7)) {isSetParameters=true}
  ]
  particleUpdateScript [
   &ParticleState(Kill Particles When Lifetime Has Elapsed=false)
   /SetVariables(Particles.Flicker:NiagaraFloat=&MakeFloatFromVector4(
     Vector 4=&MakeVector4_FromFloat(Float=&Waveform(
       Global Phase=&Add_Float(A=$System.Age, B=&MaskFloatBySpawnGroup(Masked Float=0, Passthrough Float=0.3))
       "Waveform [1]"=::NewEnumerator12
       "[1] Frequency"=30
       "[1] Amplitude Min/Max"=<0.2 1>
      ))
     Channel=::NewEnumerator0
    )) {isSetParameters=true}
   &DynamicMaterialParameters(
    Index 0 Param 1=&FloatFromCurve(
     FloatCurve=DataInterface() {propertyValues {Curve {keys [
         <at 0 5 RCIM_Cubic>
         {arriveTangent=19.5242, interpMode=RCIM_Cubic, leaveTangent=19.5242, tangentMode=RCTM_User, time=0.5, value=9}
         <at 1 14>
        ]}}}
     CurveIndex=$Particles.Flicker
    )
    Index 0 Param 2=1
    Index 0 Param 4=&RandomRangeFloat(Minimum=0, Maximum=1)
    Index 1 Param 1=200
    Index 1 Param 2=5
    Index 2 Param 1=4
    Index 2 Param 2=0.2
    Module.Param0WriteEnabled=true
    Module.Param0WriteEnabled1=true
    Module.Param0WriteEnabled2=true
    Module.Param1WriteEnabled=true
    Module.Param1WriteEnabled1=true
    Module.Param1WriteEnabled2=true
    Module.Param2WriteEnabled=true
    Module.Param2WriteEnabled1=false
    Module.Param2WriteEnabled2=false
    Module.Param3WriteEnabled=true
    Module.Param3WriteEnabled1=false
    Module.Param3WriteEnabled2=false
    Write Parameter Index 1=true
    Write Parameter Index 2=true
   )
   &SkeletalMeshLocation(
    Bone / Socket Index=&SelectIntFromArray(Direct Array Index=&SpawnGroup(), Int Selection Array=DataInterface() {propertyValues {IntData [0,1,5,6]}})
    Bone Sampling Mode=::NewEnumerator4
    MaskSpawnGroup=false
    Orientation Sampling=::NewEnumerator1
    Sampled Position Offset=&SelectVectorFromArray(
     Direct Array Index=&SpawnGroup()
     Vector Selection Array=DataInterface() {propertyValues {FloatData [<8 0 0>, <-8 0 0>, <9 2 0>, <-9 -2 0>]}}
    )
    Skeletal Mesh=$User.SkeletalMesh
   ) {module=SkeletalMeshLocation001}
   &ScaleSpriteSize(Scale Sprite Size Mode=::NewEnumerator0, Uniform Scale Factor=&Lerp_Float(A=0.3, B=1, Alpha=$Particles.Flicker))
   /SetVariables(Particles.LightColor=&MultiplyLinearColor(
     Linear Color=<rgba 0 10.2629 20 1>
     Scale Factor=&MakeVector4_FromFloat(Float=&Lerp_Float(A=0.1, B=1, Alpha=$Particles.Flicker))
    )) {isSetParameters=true}
  ]
  rendererClasses [&NiagaraLightRendererProperties, &NiagaraSpriteRendererProperties]
  renderers [
   {rendererClass=&NiagaraLightRendererProperties, rendererIndex=0}
   {rendererClass=&NiagaraSpriteRendererProperties, rendererIndex=1}
   {rendererClass=&NiagaraSpriteRendererProperties, rendererIndex=2}
  ]
 }
 MainRibbon {
  emitterUpdateScript [
   /SetVariables(Emitter.EmitterRandom:NiagaraFloat=&RandomRangeFloat(Minimum=0, Maximum=1)) {isSetParameters=true}
   &EmitterState()
   &FN_SpawnRibbon_v2(Number of Ribbons=2, Spawn Mode=::NewEnumerator0, SpawnRate=15)
  ]
  particleSpawnScript [
   &FN_InitializeRibbon_v3(
    Enable Skeletal Position=true
    Lifetime=0.65
    Ribbon Width=17
    Sampled Position Coordinate Space:ENiagaraCoordinateSpace=::Local
    Sampled Rotation Coordinate Space:ENiagaraCoordinateSpace=::Local
    Sampled Velocity Coordinate Space:ENiagaraCoordinateSpace=::Local
   ) {module=FN_InitializeRibbon_v3001}
   /SetVariables(Particles.RandomWave:NiagaraFloat=&Waveform(
     Global Phase=&Add_Float(A=$System.Age, B=&MaskFloatBySpawnGroup(Masked Float=0, Passthrough Float=0.35))
     "Waveform [1]"=::NewEnumerator12
     "[1] Frequency"=5
     "[1] Amplitude Min/Max"=<0 1>
    )) {isSetParameters=true}
   &AddVelocity(Velocity=&Lerp_Vector(A=<0 0 240>, B=<1 1 400>, Alpha=$System.FN_Integrator001.Value))
   &AddVelocity(Velocity=&MakeVector(
     X=&Waveform(
      Global Phase=&Add_Float(
       A=$System.Age
       B=&MaskFloatBySpawnGroup(Masked Float=0, Passthrough Float=&Lerp_Float(A=0.4, B=0.2, Alpha=$Emitter.EmitterRandom))
      )
      Global Amplitude Scale=&Lerp_Float(A=50, B=150, Alpha=$Particles.RandomWave)
      "Waveform [1]"=::NewEnumerator0
      "[1] Frequency"=&Lerp_Float(A=0.7, B=1.5, Alpha=$Emitter.EmitterRandom)
      "[1] Amplitude Min/Max"=<-1 1>
     )
     Y=0
     Z=0
    )) {module=AddVelocity001}
   &SampleSkeletalMeshSkeleton(Mesh:NiagaraDataInterfaceSkeletalMesh=$User.SkeletalMesh) {enabled=false}
   &SkeletalMeshLocation(
    Bone / Socket Index=&SpawnGroup()
    Bone Sampling Mode=::NewEnumerator4
    Interpolate Mesh Sampling=false
    MaskSpawnGroup=false
    Orientation Sampling=::NewEnumerator1
    Position Sampling=::NewEnumerator1
    Sampled Position Offset=&MaskVectorBySpawnGroup(Masked Vector=<8 0 0>, Passthrough Vector=<-8 0 0>)
    Skeletal Mesh=$User.SkeletalMesh
   )
  ]
  particleUpdateScript [
   &WindForce(Air Resistance=0.05, Wind Speed=<0 0 1>, Wind Speed Scale=1500)
   &Drag(Drag=5, Use Linear Drag=true, Use Rotational Drag=false)
   &LineAttractionForce(
    Attraction Strength:NiagaraFloat=&FloatFromCurve(
     FloatCurve=DataInterface() {propertyValues {Curve {keys [{arriveTangent=2.373, interpMode=RCIM_Cubic, time=0, value=0}, {interpMode=RCIM_Cubic, tangentMode=RCTM_User, time=1, value=1}]}}}
     Scale Curve=500
    )
    Line End:Vector3f=&Add_Vector(A=$Particles.SampledPosition, B=<0 0 500>)
    Line Start:Vector3f=$Particles.SampledPosition
    Use Attraction Falloff=false
   )
   &SolveForcesAndVelocity(Clamp Velocity=false, Limit Acceleration=false)
   &FN_UpdateRibbon(
    Enable Skeletal Position=true
    Initial Color:LinearColor=$Particles.Initial.Color
    Initial RibbonWidth:NiagaraFloat=$Particles.Initial.RibbonWidth
   )
   &Material(Base Texture Time:NiagaraFloat=&Multiply_Float(A=$Emitter.Age, B=-1), Jitter Amount=0)
   /SetVariables(Particles.Alpha:NiagaraFloat=&FloatFromCurve(
     FloatCurve=DataInterface() {propertyValues {Curve {keys [
         {arriveTangent=4.667, interpMode=RCIM_Cubic, leaveTangent=4.667, tangentMode=RCTM_User, time=0, value=0}
         {interpMode=RCIM_Cubic, tangentMode=RCTM_User, time=0.25, value=1}
         {tangentMode=RCTM_User, time=1, value=0}
        ]}}}
     Scale Curve=&FloatFromCurve(
      FloatCurve=DataInterface() {propertyValues {Curve {keys [{interpMode=RCIM_Cubic, tangentMode=RCTM_User, time=0, value=0}, {interpMode=RCIM_Cubic, tangentMode=RCTM_User, time=0.8, value=1}]}}}
      CurveIndex=&Waveform(
       Global Input Position=$Particles.NormalizedAge
       Global Phase=&Add_Float(
        A=&Multiply_Float(A=$System.Age, B=&Lerp_Float(A=-4, B=-6, Alpha=$Emitter.EmitterRandom))
        B=&MaskFloatBySpawnGroup(Masked Float=0, Passthrough Float=&Lerp_Float(A=0.5, B=2, Alpha=$Emitter.EmitterRandom))
       )
       "Waveform [1]"=::NewEnumerator0
       "[1] Frequency"=0.2
       "[1] Amplitude Min/Max"=<0 1>
      )
     )
    )) {isSetParameters=true}
   &ScaleColor(
    Scale Alpha=$Particles.Alpha
    Scale RGB=&MakeVectorFromLinearColorRGB(Color=&LerpLinearColors(StartColor=<rgba 0 0.4227 0.7 1>, EndColor=<rgba 0 0.5402 0.6 1>, LerpFactor=$Particles.Alpha))
    ScaleA=true
    ScaleRGB=true
   )
  ]
  rendererClasses [&NiagaraRibbonRendererProperties]
  renderers [{rendererClass=&NiagaraRibbonRendererProperties}]
 }
 RibbonCrackle {
  emitterUpdateScript [
   &EmitterState()
   &SpawnParticlesFromOtherEmitter(
    Attribute Reader=DataInterface() {propertyValues {EmitterBinding {bindingMode=Other, emitterName=MainRibbon}}}
    Spawn Rate=75
    Use Maximum Age=false
    Use Minimum Age=false
   )
  ]
  particleSpawnScript [
   /SetVariables(Particles.SizeLerp:NiagaraFloat=&FloatFromCurve(
     FloatCurve=DataInterface() {propertyValues {Curve {keys [
         {interpMode=RCIM_Cubic, tangentMode=RCTM_User, time=0, value=0}
         {arriveTangent=1.4324, interpMode=RCIM_Cubic, leaveTangent=1.4324, tangentMode=RCTM_User, time=0.87, value=0.5596}
         {interpMode=RCIM_Cubic, tangentMode=RCTM_User, time=1, value=1}
        ]}}}
     CurveIndex=&RandomRangeFloat(Minimum=0, Maximum=1)
    )) {isSetParameters=true}
   &InitializeParticle(
    Color=&SelectLinearColorFromArray(
     Array Sampling Mode=::NewEnumerator1
     Direct Array Index=&MakeCustomIntFromBool(Bool=&SetBoolByFloatComparison(A=$Particles.SizeLerp, B=0.1), True INT=1, False INT=0)
     Color Selection Array=DataInterface() {propertyValues {ColorData [<rgba 0 1.2486 2 1>, <rgba 0.7305 0 0.1823 1>]}}
     Random Seed=?
    )
    Lifetime Max=0.3
    Lifetime Min=0.1
    Lifetime Mode=::NewEnumerator1
    Sprite Rotation Mode=::NewEnumerator2
    Sprite Size Mode=::NewEnumerator3
    Uniform Sprite Size=&Lerp_Float(A=4, B=5.5, Alpha=$Particles.SizeLerp)
    UsePositionOffset=false
    Write Material Random=true
   )
   &SampleParticlesFromOtherEmitter(
    Color Sampling=::NewEnumerator1
    Maximum Sampled Particle Age=0.35
    Module.MaxSampledAgeCutoff=true
    Module.MinSampledAgeCutoff=false
    Reject Particle By Sampled Age=true
    Velocity Sampling=::NewEnumerator0
   )
   /SetVariables(Particles.Intensity:NiagaraFloat=&FloatFromCurve(
     FloatCurve=DataInterface() {propertyValues {Curve {keys [
         {arriveTangent=4.6871, interpMode=RCIM_Cubic, leaveTangent=4.6871, tangentMode=RCTM_User, time=0, value=0}
         {interpMode=RCIM_Cubic, tangentMode=RCTM_User, time=0.7, value=1}
        ]}}}
     CurveIndex=&MakeFloatFromLinearColor(LinearColor=$Output.SampleParticlesFromOtherEmitter.SampledColor, Channel:ENiagaraLinearColor_Channels=::NewEnumerator3)
    )) {isSetParameters=true}
   &CameraOffset(Camera Offset Amount=&RandomRangeFloat(Minimum=-5, Maximum=5))
  ]
  particleUpdateScript [
   &ParticleState()
   &Drag(Drag=3, Use Linear Drag=true, Use Rotational Drag=false)
   &ScaleSpriteSize(
    Uniform Curve Scale=&Multiply_Float(
     A=$Particles.Intensity
     B=&Waveform(
      Global Phase=$Particles.Age
      "Waveform [1]"=::NewEnumerator12
      "[1] Frequency"=&RandomRangeFloat(Minimum=25, Maximum=50)
      "[1] Amplitude Min/Max"=<0.5 1.5>
     )
    )
    Uniform Curve Sprite Scale=DataInterface() {propertyValues {Curve {keys [
        {arriveTangent=10.8984, interpMode=RCIM_Cubic, leaveTangent=10.8984, tangentMode=RCTM_User, time=0, value=0}
        <at 0.2 1>
        {interpMode=RCIM_Cubic, tangentMode=RCTM_User, time=0.6, value=1}
        {arriveTangent=-5.0292, interpMode=RCIM_Cubic, leaveTangent=-5.0292, tangentMode=RCTM_User, time=1, value=0}
       ]}}}
   )
   /SetVariables(
    Particles.SpriteParams0:Vector4f=&MakeFloat4FromVectorAndFloat("Vector (RGB)"=&MakeVectorFromLinearColorRGB(Color=<rgba 0 1.7271 2 1>), "Float (Alpha)"=1)
    Particles.SpriteParams1:Vector4f=<0 0 -1 0>
   ) {isSetParameters=true, module=SetVariables_#3}
   &SolveForcesAndVelocity(Clamp Velocity=false, Limit Acceleration=false)
  ]
  rendererClasses [&NiagaraSpriteRendererProperties]
  renderers [{rendererClass=&NiagaraSpriteRendererProperties}]
 }
 Splines001 {
  emitterUpdateScript [
   &EmitterState()
   &SpawnRate(Spawn Probability=0.4, SpawnRate=&Lerp_Float(A=0, B=20, Alpha=$System.Dive), Use Spawn Probability=true)
  ]
  particleSpawnScript [
   &FN_SetSpawnGroup(Number of Spawn Groups=2)
   /SetVariables(Particles.ColorID:NiagaraInt32=&RandomRangeInt(Minimum=0, Maximum=1)) {isSetParameters=true}
   &InitializeParticle(
    Color=&SelectLinearColorFromArray(
     Array Sampling Mode=::NewEnumerator1
     Direct Array Index=$Particles.ColorID
     Color Selection Array=DataInterface() {propertyValues {ColorData [<rgba 1 0.0084 0.9721 1>, <rgba 0 1 1 1>]}}
     Random Seed=?
    )
    Lifetime Max=0.6
    Lifetime Min=0.16
    Lifetime Mode=::NewEnumerator1
    Mesh Scale Max=<2.5 2.5 3.5>
    Mesh Scale Min=<1.5 1.5 2>
    Mesh Scale Mode=::NewEnumerator2
    UsePositionOffset=false
    Write Material Random=true
   )
   &InitialMeshOrientation(
    Mesh Axis To Orient=<0 0 1>
    Mesh Orientation Mode=::NewEnumerator2
    Orientation Vector=<0 0 -1>
    Rotation=&RandomRangeVector(Minimum=<0 0 0>, Maximum=<0 0 1>)
    Use Rotation Vector=true
   )
   &CameraOffset(Camera Offset Amount=5)
   &SkeletalMeshLocation(
    Bone / Socket Index=&SpawnGroup()
    Bone Sampling Mode=::NewEnumerator4
    Interpolate Mesh Sampling=false
    MaskSpawnGroup=false
    Orientation Sampling=::NewEnumerator1
    Sampled Position Offset=&MaskVectorBySpawnGroup(Masked Vector=<10 0 0>, Passthrough Vector=<-10 0 0>)
    Skeletal Mesh=$User.SkeletalMesh
   ) {module=SkeletalMeshLocation001}
  ]
  particleUpdateScript [
   &ParticleState()
   &DynamicMaterialParameters(
    Index 0 Param 1=&FloatFromCurve(
     FloatCurve=DataInterface() {propertyValues {Curve {keys [
         {interpMode=RCIM_Cubic, tangentMode=RCTM_User, time=0, value=1}
         {arriveTangent=-3.0019, interpMode=RCIM_Cubic, leaveTangent=-3.0019, tangentMode=RCTM_User, time=1, value=0}
        ]}}}
     Scale Curve=&RandomRangeFloat(Minimum=6, Maximum=16)
    )
    Index 0 Param 2=&RandomRangeFloat(Minimum=-5, Maximum=-3)
    Index 0 Param 3=&FloatFromCurve(FloatCurve=DataInterface() {propertyValues {Curve {keys [<at 0 1 RCIM_Cubic>, {arriveTangent=-0.8028, interpMode=RCIM_Cubic, leaveTangent=-0.8028, tangentMode=RCTM_User, time=1, value=0}]}}})
    Index 0 Param 4=&MaskFloatBySpawnGroup(Masked Float=0, Passthrough Float=0.5)
    Module.Param0WriteEnabled=true
    Module.Param1WriteEnabled=true
    Module.Param2WriteEnabled=true
    Module.Param3WriteEnabled=true
   )
   &UpdateMeshOrientation(Rotation Rate=&RandomRangeFloat(Minimum=-0.5, Maximum=0.5), Rotation Vector=<0 0 1>)
   &ScaleMeshSize(Scale Factor=&VectorFromCurve(VectorCurve=DataInterface() {propertyValues {
       XCurve {keys [
         {arriveTangent=2.2128, interpMode=RCIM_Cubic, leaveTangent=2.2128, tangentMode=RCTM_User, time=0, value=1}
         {arriveTangent=-0.9, interpMode=RCIM_Cubic, leaveTangent=-0.9, time=0.2, value=1.4}
         {arriveTangent=-2.7027, interpMode=RCIM_Cubic, leaveTangent=-2.7027, tangentMode=RCTM_User, time=1, value=0.1}
        ]}
       YCurve {keys [
         {arriveTangent=2.2128, interpMode=RCIM_Cubic, leaveTangent=2.2128, tangentMode=RCTM_User, time=0, value=1}
         {arriveTangent=-0.9, interpMode=RCIM_Cubic, leaveTangent=-0.9, time=0.2, value=1.4}
         {arriveTangent=-2.7027, interpMode=RCIM_Cubic, leaveTangent=-2.7027, tangentMode=RCTM_User, time=1, value=0.1}
        ]}
       ZCurve {keys [<at 0 0.4>,<at 1 1>]}
      }}))
   &ScaleColor(
    Scale Alpha=&FloatFromCurve(FloatCurve=DataInterface() {propertyValues {Curve {keys [<at 0 1 RCIM_Cubic>, {arriveTangent=-2.2348, interpMode=RCIM_Cubic, leaveTangent=-2.2348, tangentMode=RCTM_User, time=1, value=0}]}}})
    ScaleA=true
    ScaleRGB=false
   )
   /SetVariables(Particles.Color2=&SelectLinearColorFromArray(
     Array Sampling Mode=::NewEnumerator1
     Direct Array Index=$Particles.ColorID
     Color Selection Array=DataInterface() {propertyValues {ColorData [<rgba 0.0285 0.9303 1 0.1>, <rgba 1 0.0152 0.1852 0.05>]}}
     Random Seed=?
    )) {isSetParameters=true}
   /SetVariables(Particles.Position2:NiagaraPosition=&Add_Vector(A=Linked() {linkedVariable {name=Particles.Position, type=NiagaraPosition}}, B=<2 2 1>)) {isSetParameters=true}
  ]
  rendererClasses [&NiagaraMeshRendererProperties]
  renderers [
   {rendererClass=&NiagaraMeshRendererProperties, rendererIndex=0}
   {rendererClass=&NiagaraMeshRendererProperties, rendererIndex=1}
  ]
 }
}
summary {
 systemName=NS_Contrail_Dive_RoseDust
 emitters [
  {emitterName=MainRibbon, rendererClasses [&NiagaraRibbonRendererProperties]}
  {emitterName=RibbonCrackle, rendererClasses [&NiagaraSpriteRendererProperties]}
  {emitterName=HandCrackle, rendererClasses [&NiagaraSpriteRendererProperties]}
  {emitterName=HandCrackleInner, rendererClasses [&NiagaraSpriteRendererProperties]}
  {emitterName=HandGlow, rendererClasses [&NiagaraLightRendererProperties, &NiagaraSpriteRendererProperties]}
  {emitterName=Splines001, rendererClasses [&NiagaraMeshRendererProperties]}
  {emitterName=BodyCrackle, rendererClasses [&NiagaraSpriteRendererProperties]}
  {emitterName=GlitchShapes, rendererClasses [&NiagaraMeshRendererProperties]}
 ]
 userVariables [
  User.bIsFrontEnd {type=NiagaraBool}
  User.bIsFrontEndPreview {type=NiagaraBool}
  User.RibbonIDOffset {type=NiagaraInt32}
  User.SkeletalMesh {
   type=NiagaraDataInterfaceSkeletalMesh
   defaultValue {
    struct=NiagaraExt_VariableValue_DataInterface
    value {dataInterface=&self:NiagaraDataInterfaceSkeletalMesh_1}
   }
  }
 ]
}
